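# Unlit two-sided additive trail material for Niagara ribbons and beam trails that draws a tire burnout glow.
# An inverted burnout mask on UV1 multiplies particle color and a UV0 U-axis fade, shaped by dynamic parameters for reveal power and contrast.
# Color passes through MF_FX_EmissiveAdjustments.
# Alpha gets a depth fade with mask-driven distance, skipped on mobile and low quality.
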Material /Game/Effects/Fort_Effects/Materials/Trails/M_CheapTireGlow
graph {
 properties {
  bUsedWithBeamTrails=true
  bUsedWithNiagaraRibbons=true
  blendMode=BLEND_Additive
  pixelDepthOffsetMode=PDOM_Legacy
  shadingModel=MSM_Unlit
  twoSided=true
  usedShadingModels=Unlit
 }
 nodes [
  n1 = TextureSample(UVs=TextureCoordinate() {coordinateIndex=1}) {texture=/Game/Effects/Fort_Effects/Textures/Masks/T_TireBurnOut}
  n2 = ParticleColor()
  n3 = ComponentMask(None=TextureCoordinate()) {r=true}
  n4 = DynamicParameter() {defaultValue=<rgba 4 24 1 1>, paramNames [PeelOutRevealPower, ContrastTweakBurnOutColor, Param3, Param4]}
  n5 = LinearInterpolate(B=@n4.ContrastTweakBurnOutColor, Alpha=@n3) {constB=24}
  n6 = DynamicParameter() {defaultValue=<rgba 4 24 1 1>, paramNames [PeelOutRevealPower, ContrastTweakBurnOutColor, Param3, Param4]}
  n7 = Power(Base=Multiply(A=OneMinus(None=@n3), B=Clamp(None=@n5)), Exp=@n6.PeelOutRevealPower)
  MaterialFunctionCall(Emissive=Multiply(A=OneMinus(None=@n1.RGB), B=Multiply(A=@n2.RGB, B=@n7))) {materialFunction=&MF_FX_EmissiveAdjustments}
  n8 = Multiply(A=OneMinus(None=@n1.A), B=Multiply(A=@n2.A, B=@n7))
  n9 = LinearInterpolate(Alpha=@n1.A) {constA=6, constB=12}
  QualitySwitch(Default=ShadingPathSwitch(Default=DepthFade(Opacity=@n8, FadeDistance=@n9), Mobile=@n8) {desc=Migrated from FeatureLevelSwitch}, Low=@n8)
 ]
}
