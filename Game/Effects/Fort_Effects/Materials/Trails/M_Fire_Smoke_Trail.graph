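# Unlit, two-sided translucent material for a glowing fire and smoke trail on Niagara ribbons, beam trails and sprites.
# Opacity combines panning smoke masks, dynamic-parameter erosion, end, depth and near-camera fades, plus optional view-angle fade.
# Emissive blends two glow colors using a trail-length mask and fake normal-based shading, then adds a volcano smoke color function.
# Low quality and mobile replace detail maps with constants; mobile also skips depth fade.
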Material /Game/Effects/Fort_Effects/Materials/Trails/M_Fire_Smoke_Trail
graph {
 properties {
  bUsedWithBeamTrails=true
  bUsedWithNiagaraRibbons=true
  bUsedWithParticleSprites=true
  blendMode=BLEND_Translucent
  pixelDepthOffsetMode=PDOM_Legacy
  shadingModel=MSM_Unlit
  twoSided=true
  usedShadingModels=Unlit
 }
 nodes [
  n1 = ParticleColor()
  n2 = DepthFade(FadeDistance=ScalarParameter() {defaultValue=300, parameterName=DepthFadeDist})
  n3 = Constant() {r=1}
  n4 = Multiply(
   A=MaterialFunctionCall(Distance=ScalarParameter() {defaultValue=200, parameterName=NearCamFadeDistance}) {materialFunction=&Near_Camera_Fade}.Result
   B=FeatureLevelSwitch(Default=@n2, ES2=@n3, ES3_1=@n3, SM4_REMOVED=@n2)
  )
  n5 = DynamicParameter() {defaultValue=<rgba 1 4 1 1>, paramNames [PanSpeed,GlowPower,Param3,Erode]}
  n6 = Clamp(None=@n5.Erode)
  n7 = Multiply(A=@n6) {constB=-1}
  n8 = TextureCoordinate() {coordinateIndex=1}
  n9 = ScalarParameter() {defaultValue=1, parameterName=Tile_X}
  n10 = ScalarParameter() {defaultValue=1, parameterName=Tile_Y}
  n11 = Multiply(A=Multiply(A=Time(), B=ScalarParameter() {defaultValue=1, parameterName=PanSpeed}), B=@n5.PanSpeed)
  n12 = Panner(Coordinate=Multiply(A=@n8, B=AppendVector(A=@n9, B=@n10)), Time=@n11) {speedX=0.5}
  n13 = TextureSample(UVs=@n12) {texture=/Game/Effects/Fort_Effects/Gameplay/SmokeDustExplosions/Textures/T_RibbonSmokeMasked_04}
  n14 = Add(A=@n6, B=@n13.R)
  n15 = Multiply(A=Saturate(None=LinearInterpolate(A=@n7, B=@n14, Alpha=@n6)), B=@n13.G)
  n16 = Constant() {r=1}
  n17 = Constant() {r=0.5}
  n18 = SphereMask(
   A=ComponentMask(None=TextureCoordinate()) {r=true}
   B=@n17
   Radius=@n17
   Hardness=ScalarParameter() {defaultValue=0.8, parameterName=EndFadeHardness}
  ) {hardnessPercent=80}
  n19 = ScalarParameter() {defaultValue=1, parameterName=UseEndFadeHardness}
  n20 = Multiply(A=@n1.A, B=Multiply(A=Multiply(A=@n15, B=@n4), B=LinearInterpolate(A=@n16, B=@n18, Alpha=@n19)))
  n21 = ScalarParameter() {parameterName=AngleFadeSensitivity}
  LinearInterpolate(
   A=@n20
   B=Multiply(
    A=@n20
    B=MaterialFunctionCall("FadeSensitivity (optional)"=@n21) {materialFunction=&FadeBasedOnViewAngle}.Result
   )
   Alpha=ScalarParameter() {parameterName=UseAngleFade}
  )
  n22 = VectorParameter() {group=Color, parameterName=GlowColor2, defaultValue=<rgba 1.5 0.015 0 0>}
  n23 = VectorParameter() {group=Color, parameterName=GlowColor1, defaultValue=<rgba 3 0.731 0 0>}
  n24 = ComponentMask(None=TextureCoordinate()) {r=true}
  n25 = Power(
   Base=LinearInterpolate(
    A=OneMinus(None=@n24)
    B=DynamicParameter() {parameterIndex=1, paramNames [EmissvieArea,Param2,Param3,Param4]}.EmissvieArea
    Alpha=ScalarParameter() {parameterName=UseDynamicParamToHandelEmissArea}
   )
   Exp=ScalarParameter() {defaultValue=2, group=Color, parameterName=GlowMaskPower}
  )
  n26 = VectorParameter() {parameterName=DetailNormalMult, defaultValue=<rgba 1 1 0 0>}
  n27 = ScalarParameter() {defaultValue=1.5, parameterName=DetailTileX}
  n28 = ScalarParameter() {defaultValue=0.75, parameterName=DetailTileY}
  n29 = Panner(
   Coordinate=Multiply(A=AppendVector(A=@n27, B=@n28), B=@n8)
   Time=Multiply(A=ScalarParameter() {defaultValue=-0.5, parameterName=DetailPanMult}, B=@n11)
  ) {speedX=0.5, speedY=0.04}
  n30 = TextureSample(UVs=@n29) {samplerType=SAMPLERTYPE_Normal, texture=/Game/Global/Textures/Default/Tileable_Patterns/T_StormCloudDetail01_N}
  n31 = Constant()
  n32 = QualitySwitch(Default=Multiply(A=Multiply(A=@n26.RGB, B=@n30.RGB), B=OneMinus(None=@n13.R)), Low=@n31)
  n33 = Add(
   A=FeatureLevelSwitch(Default=@n32, ES2=@n31, ES3_1=@n31, SM4_REMOVED=@n32)
   B=TextureSample(UVs=@n12) {samplerType=SAMPLERTYPE_Normal, texture=/Game/Effects/Fort_Effects/Gameplay/SmokeDustExplosions/Textures/T_RibbonSmokeMasked_01}.RGB
  )
  n34 = DotProduct(A=Constant3Vector() {constant=<rgba -1 0 0 0>}, B=@n33)
  n35 = OneMinus(None=TextureSample(UVs=@n29) {samplerType=SAMPLERTYPE_LinearGrayscale, texture=/Game/Global/Textures/Default/Tileable_Patterns/T_StormCloudDetail01_M}.G)
  n36 = Constant() {r=0.2}
  n37 = QualitySwitch(Default=Multiply(A=@n35, B=@n35), Low=@n36)
  n38 = FeatureLevelSwitch(Default=@n37, ES2=@n36, ES3_1=@n36, SM4_REMOVED=@n37)
  n39 = Add(
   A=Multiply(A=@n25, B=ScalarParameter() {defaultValue=0.5, group=Color, parameterName=GlowMaskAdd}) {constB=0.5}
   B=LinearInterpolate(A=Saturate(None=@n34), B=LinearInterpolate(A=@n38, Alpha=@n13.B) {constA=0.2}) {constAlpha=0.6}
  ) {constB=0.2}
  n40 = Power(Base=@n39, Exp=@n5.GlowPower) {constExponent=3}
  n41 = Multiply(A=@n25, B=Multiply(A=LinearInterpolate(A=@n22.RGB, B=@n23.RGB, Alpha=@n40), B=@n39))
  n42 = Transform(None=@n33)
  MaterialFunctionCall(Emissive=Add(
    A=Multiply(A=@n1.RGB, B=@n41)
    B=MaterialFunctionCall!(WorldSpaceNormal=@n42, MaskForLowEnd=@n13.R) {materialFunction=/Game/Athena/Prototype/Blueprints/White/FX/MF_Volcano_Smoke_Colors}.Result
   )) {materialFunction=&MF_FX_EmissiveAdjustments}
 ]
}
