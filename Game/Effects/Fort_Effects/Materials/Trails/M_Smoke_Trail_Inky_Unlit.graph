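# Unlit translucent smoke trail material for beam trails, Niagara ribbons and sprites, with inky shading that fakes sun/moon lighting.
# A panned normal map dotted with the global light direction, times eroded smoke texture, blends tinted, desaturated particle color toward full color; mobile uses constant lighting.
# Dynamic parameters drive erosion, pan speed, optional sphere mask and master alpha.
# Opacity adds edge masking, optional banding, and non-mobile depth and camera fades.
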Material /Game/Effects/Fort_Effects/Materials/Trails/M_Smoke_Trail_Inky_Unlit
graph {
 properties {
  bUsedWithBeamTrails=true
  bUsedWithNiagaraRibbons=true
  bUsedWithParticleSprites=true
  blendMode=BLEND_Translucent
  pixelDepthOffsetMode=PDOM_Legacy
  shadingModel=MSM_Unlit
  usedShadingModels=Unlit
 }
 nodes [
  n1 = ParticleColor()
  n2 = Constant() {r=0.5}
  n3 = ScalarParameter() {defaultValue=1, parameterName=UMultiply}
  n4 = ScalarParameter() {defaultValue=1, parameterName=VMultiply}
  n5 = DynamicParameter() {defaultValue=<rgba 1 0 1 1>, paramNames [Erode, SphereMaskRadius, PanSpeed, MasterAlpha]}
  n6 = Panner(
   Coordinate=Multiply(A=AppendVector(A=@n3, B=@n4), B=TextureCoordinate() {coordinateIndex=1})
   Time=Multiply(A=@n5.PanSpeed, B=Time())
   Speed=AppendVector(A=ScalarParameter() {defaultValue=0.6, parameterName=U_PannerSpeed}, B=Constant())
  ) {speedX=0.6}
  n7 = TextureSampleParameter2D(UVs=@n6) {parameterName=TrailNormal, samplerType=SAMPLERTYPE_Normal, texture=/Game/Effects/Fort_Effects/Textures/Smoke/T_Smoke_Trail_Soft_N}
  n8 = Transform(None=Multiply(A=@n7.RGB, B=Constant3Vector() {constant=<rgba -1 -1 1 0>}))
  n9 = CollectionParameter() {
   collection=&FortniteMaterialParameters
   parameterName=SunAndMoonModelDirectionalVector
  }
  n10 = ComponentMask(None=@n9) {b=true, g=true, r=true}
  n11 = Clamp(None=Abs(None=DotProduct(A=@n8, B=@n10)))
  n12 = Constant() {r=0.5}
  n13 = TextureSampleParameter2D(UVs=@n6) {parameterName=TrailTexture, samplerType=SAMPLERTYPE_Grayscale, texture=&T_Smoke_Trail_Soft}
  n14 = Add(A=@n5.Erode, B=@n13.G)
  n15 = Clamp(None=Multiply(A=LinearInterpolate(Alpha=@n14) {constA=-3}, B=@n13.G))
  n16 = LinearInterpolate(
   A=Multiply(A=VectorParameter() {parameterName=DarkColor, defaultValue=<rgba 0.0896 0.1113 0.2 0>}.RGB, B=Desaturation(None=@n1.RGB, Fraction=@n2))
   B=@n1.RGB
   Alpha=Multiply(A=ShadingPathSwitch(Default=@n11, Mobile=@n12) {desc=Migrated from FeatureLevelSwitch}, B=@n15)
  )
  n17 = SphereMask(
   A=WorldPosition().XYZ
   B=ObjectPositionWS()
   Radius=DynamicParameter() {defaultValue=<rgba 1 0 1 1>, paramNames [Erode, SphereMaskRadius, PanSpeed, MasterAlpha]}.SphereMaskRadius
   Hardness=Constant()
  )
  MaterialFunctionCall(Emissive=StaticSwitchParameter(True=Multiply(A=@n16, B=@n17), False=@n16) {parameterName=UseDP1asLocalSphereAlphaMask}) {materialFunction=&MF_FX_EmissiveAdjustments}
  n18 = ChannelMaskParameter(None=TextureCoordinate()) {parameterName=Channel Mask Selector}
  Multiply(A=@n18.R) {constB=0.5}
  n19 = MaterialFunctionCall(Value=@n18.R, Period=Constant() {r=2}) {materialFunction=&LinearSine}
  n20 = ScalarParameter() {defaultValue=5, parameterName=LinearBandAlphaPower}
  n21 = TextureCoordinate()
  n22 = Constant() {r=0.5}
  n23 = Multiply(A=@n1.A, B=@n15)
  n24 = ScalarParameter() {defaultValue=150, parameterName=Fade Length}
  n25 = ScalarParameter() {defaultValue=25, parameterName=Fade Offset}
  n26 = DepthFade(
   Opacity=Multiply(
    A=@n23
    B=MaterialFunctionCall(Fade Length=@n24, Fade Offset=@n25) {materialFunction=&CameraDepthFade}.Result
   )
   FadeDistance=ScalarParameter() {defaultValue=40, parameterName=DepthFadeDist}
  ) {fadeDistanceDefault=40}
  n27 = Multiply(
   A=SphereMask(A=ComponentMask(None=@n21) {r=true}, B=@n22, Radius=@n22) {attenuationRadius=0.5, hardnessPercent=90}
   B=ShadingPathSwitch(Default=@n26, Deferred=@n26, Mobile=@n23) {desc=Migrated from FeatureLevelSwitch}
  )
  n28 = StaticSwitchParameter(
   True=Multiply(A=@n27, B=DynamicParameter() {defaultValue=<rgba 1 0 1 1>, paramNames [Erode, SphereMaskRadius, PanSpeed, MasterAlpha]}.MasterAlpha)
   False=@n27
  ) {parameterName=UseDP3asMasterAlpha}
  n29 = StaticSwitchParameter(True=Multiply(A=@n28, B=@n17), False=@n28) {parameterName=UseDP1asLocalSphereAlphaMask}
  StaticSwitchParameter(True=Multiply(A=Power(Base=@n19.Linear Sine, Exp=@n20) {constExponent=4}, B=@n29), False=@n29) {parameterName=UseLinearGradBandAlpha}
 ]
}
