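# Translucent, two-sided smoke trail material for beam trails, Niagara ribbons and mesh particles.
# A scaled, rotatable, panning smoke texture is eroded through dynamic parameters, feeding opacity and HDR emissive from particle color.
# A UseFadeStart switch picks a sine fade along rotated U or a squared falloff toward the end.
# Time-of-day tinting is switchable, and depth fade softens opacity except on mobile.
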Material /Game/Effects/Fort_Effects/Materials/Trails/M_Smoke_Trail_Soft_FadeStart_2
graph {
 properties {
  bUsedWithBeamTrails=true
  bUsedWithNiagaraMeshParticles=true
  bUsedWithNiagaraRibbons=true
  blendMode=BLEND_Translucent
  pixelDepthOffsetMode=PDOM_Legacy
  twoSided=true
  usedShadingModels=Default Lit
 }
 nodes [
  n1 = VectorParameter() {parameterName=DayTimeColor, defaultValue=<rgba 1 1 1 0>}
  n2 = VectorParameter() {parameterName=NightTimeColor, defaultValue=<rgba 1 1 1 0>}
  n3 = MaterialFunctionCall(Color_Morning=@n1.RGB, Color_Day=@n1.RGB, Color_Evening=@n2.RGB, Color_Night=@n2.RGB) {materialFunction=&Color_Time_of_Day}
  n4 = ParticleColor()
  n5 = DynamicParameter() {defaultValue=<rgba 1 0 1 1>, paramNames [Erode, EmissiveAmount, PanSpeed, MasterAlpha]}
  n6 = DynamicParameter() {defaultValue=<rgba 1 0 1 1>, paramNames [Erode, EmissiveAmount, PanSpeed, MasterAlpha]}
  n7 = AppendVector(A=ScalarParameter() {defaultValue=0.3, parameterName=U}, B=ScalarParameter() {defaultValue=1, parameterName=V})
  n8 = MaterialFunctionCall(UVs=TextureCoordinate(), "Rotation Angle (0-1)"=ScalarParameter() {parameterName="UV_Rotation Angle (0-1) (S)"}) {materialFunction=&CustomRotator}
  n9 = ScalarParameter() {defaultValue=-0.3, parameterName=PanSpeed}
  n10 = Constant()
  n11 = Time()
  n12 = TextureSampleParameter2D(UVs=Panner(Coordinate=Multiply(A=@n7, B=@n8.Rotated Values), Time=Multiply(A=@n6.PanSpeed, B=@n11), Speed=AppendVector(A=@n9, B=@n10)) {speedX=0.6}) {parameterName=TrailTexture_BChannel, samplerType=SAMPLERTYPE_Grayscale, texture=&T_Smoke_Trail_Soft}
  n13 = Add(A=@n6.Erode, B=@n12.G)
  n14 = Clamp(None=Multiply(A=LinearInterpolate(Alpha=@n13) {constA=-3}, B=@n12.G))
  n15 = MaterialFunctionCall(LDR Input=Clamp(None=@n4.RGB), HDR Tint=Multiply(A=@n4.RGB, B=Multiply(A=@n5.EmissiveAmount, B=@n14))) {materialFunction=&DeriveHDRfromLDR}
  MaterialFunctionCall(Emissive=StaticSwitchParameter(True=Multiply(A=@n3.Result, B=@n15.RemappedRange), False=@n15.RemappedRange) {parameterName=UseTimeOfDay}) {materialFunction=&MF_FX_EmissiveAdjustments}
  Reroute(None=@n4.RGB)
  n16 = MaterialFunctionCall(Float2=@n8.Rotated Values) {materialFunction=&BreakOutFloat2Components}
  n17 = Multiply(A=Sine(None=Multiply(A=@n16.R) {constB=0.5})) {constB=1.5}
  n18 = ComponentMask(None=@n8.Rotated Values) {r=true}
  n19 = Clamp(None=Multiply(A=@n18, B=@n18))
  n20 = Multiply(
   A=Multiply(
    A=Multiply(A=@n4.A, B=@n14)
    B=DynamicParameter() {defaultValue=<rgba 1 0 1 1>, paramNames [Erode, EmissiveAmount, PanSpeed, MasterAlpha]}.MasterAlpha
   )
   B=StaticSwitchParameter(True=Saturate(None=@n17), False=OneMinus(None=@n19)) {parameterName=UseFadeStart}
  )
  ShadingPathSwitch(Default=DepthFade(Opacity=@n20, FadeDistance=ScalarParameter() {group=Opacity, parameterName=FadeDistance}), Mobile=@n20) {desc=Migrated from FeatureLevelSwitch}
  TextureCoordinate()
  TextureCoordinate()
  TextureCoordinate()
 ]
}
