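# Translucent Niagara ribbon material for a soft, panning smoke trail with controllable chromatic fringing.
# Three samples of the smoke texture, offset forward, none and back by dynamic Chroma_U_Scale on both axes, become RGB, tinted by particle color.
# Emissive gets HDR remapping and optional time-of-day tint.
# Opacity combines summed samples, dynamic erosion, particle and master alpha, camera depth fade and a sine fade-in along U.
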Material /Game/Effects/Fort_Effects/Materials/Trails/M_Smoke_Trail_Soft_FadeStart_ChromaticControl
graph {
 properties {bUsedWithNiagaraRibbons=true, blendMode=BLEND_Translucent, pixelDepthOffsetMode=PDOM_Legacy, usedShadingModels=Default Lit}
 nodes [
  n1 = VectorParameter() {parameterName=DayTimeColor, defaultValue=<rgba 1 1 1 0>}
  n2 = VectorParameter() {parameterName=NightTimeColor, defaultValue=<rgba 1 1 1 0>}
  n3 = MaterialFunctionCall(Color_Morning=@n1.RGB, Color_Day=@n1.RGB, Color_Evening=@n2.RGB, Color_Night=@n2.RGB) {materialFunction=&Color_Time_of_Day}
  n4 = ParticleColor()
  n5 = DynamicParameter() {parameterIndex=1, defaultValue=<rgba 1 0.01 0 1>, paramNames [U_Tile, Chroma_U_Scale, Chroma_V_Scale, ----]}
  n6 = Multiply(A=TextureCoordinate() {coordinateIndex=1}, B=AppendVector(A=@n5.U_Tile, B=Constant() {r=1}))
  n7 = AppendVector(A=@n5.Chroma_U_Scale, B=@n5.Chroma_U_Scale)
  n8 = TextureSampleParameter2D(UVs=Panner(Coordinate=Add(A=@n6, B=@n7), Time=NamedRerouteUsage(), Speed=NamedRerouteUsage()) {speedX=0.6}) {parameterName=TrailTexture_BChannel, samplerType=SAMPLERTYPE_Grayscale, texture=&T_Smoke_Trail_Soft}
  n9 = TextureSampleParameter2D(UVs=Panner(Coordinate=@n6, Time=NamedRerouteUsage(), Speed=NamedRerouteUsage()) {speedX=0.6}) {parameterName=TrailTexture_BChannel, samplerType=SAMPLERTYPE_Grayscale, texture=&T_Smoke_Trail_Soft}
  n10 = AppendVector(A=@n8.G, B=@n9.G)
  n11 = Multiply(A=@n7) {constB=-1}
  n12 = TextureSampleParameter2D(UVs=Panner(Coordinate=Add(A=@n6, B=@n11), Time=NamedRerouteUsage(), Speed=NamedRerouteUsage()) {speedX=0.6}) {parameterName=TrailTexture_BChannel, samplerType=SAMPLERTYPE_Grayscale, texture=&T_Smoke_Trail_Soft}
  n13 = Multiply(
   A=@n4.RGB
   B=Multiply(
    A=DynamicParameter() {defaultValue=<rgba 1 0 1 1>, paramNames [Erode, EmissiveAmount, PanSpeed, MasterAlpha]}.EmissiveAmount
    B=AppendVector(A=@n10, B=@n12.G)
   )
  )
  n14 = MaterialFunctionCall(LDR Input=Clamp(None=@n13), HDR Tint=@n13) {materialFunction=&DeriveHDRfromLDR}
  MaterialFunctionCall(Emissive=StaticSwitchParameter(True=Multiply(A=@n3.Result, B=@n14.RemappedRange), False=@n14.RemappedRange) {parameterName=UseTimeOfDay}) {materialFunction=&MF_FX_EmissiveAdjustments}
  n15 = DynamicParameter() {defaultValue=<rgba 1 0 1 1>, paramNames [Erode, EmissiveAmount, PanSpeed, MasterAlpha]}
  n16 = LinearInterpolate(Alpha=Add(A=@n15.Erode, B=@n9.G)) {constA=-3}
  n17 = Saturate(None=Add(A=Add(A=@n9.G, B=@n8.G), B=@n12.G))
  n18 = Multiply(A=@n4.A, B=Clamp(None=Multiply(A=@n16, B=@n17)))
  n19 = MaterialFunctionCall(
   Fade Length=ScalarParameter() {defaultValue=512, parameterName=Fade Length}
   Fade Offset=ScalarParameter() {defaultValue=24, parameterName=Fade Offset}
  ) {materialFunction=&CameraDepthFade}
  n20 = Multiply(A=MaterialFunctionCall(Float2=TextureCoordinate()) {materialFunction=&BreakOutFloat2Components}.R) {constB=0.5}
  n21 = Saturate(None=Multiply(A=Sine(None=@n20)) {constB=1.5})
  Multiply(
   A=Multiply(
    A=Multiply(A=@n18, B=@n19.Result)
    B=DynamicParameter() {defaultValue=<rgba 1 0 1 1>, paramNames [Erode, EmissiveAmount, PanSpeed, MasterAlpha]}.MasterAlpha
   )
   B=@n21
  )
  NamedRerouteDeclaration(None=AppendVector(A=ScalarParameter() {defaultValue=0.6, parameterName=PanSpeed}, B=Constant())) {name=Panner Speed, nodeColor=<rgba 0 0.3672 1 1>}
  NamedRerouteDeclaration(None=Multiply(A=@n15.PanSpeed, B=Time())) {name=TimeMulti, nodeColor=<rgba 0.7578 0 1 1>}
  Reroute(None=@n4.RGB)
  n22 = ComponentMask(None=TextureCoordinate()) {r=true}
  n23 = Multiply(A=@n22, B=@n22)
  StaticSwitchParameter(True=@n21, False=OneMinus(None=Clamp(None=@n23))) {parameterName=UseFadeStart}
 ]
}
