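# Translucent, lit, two-sided smoke trail material for Niagara ribbons and mesh particles.
# Two panning wispy-noise layers, the second distorting the first, build a masked smoke shape with smoothstep erosion, edge falloff and depth fade off mobile.
# Particle color or a two-color gradient tints the emissive, while dynamic parameters control erosion, tiling, pan offset and width.
# A UseSpline switch swaps UVs, masks and normals and adds SplineThicken vertex offset.
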Material /Game/Effects/Fort_Effects/Materials/Trails/M_Smooth_Smoke_Ribbon
graph {
 properties {
  bUsedWithNiagaraMeshParticles=true
  bUsedWithNiagaraRibbons=true
  blendMode=BLEND_Translucent
  pixelDepthOffsetMode=PDOM_Legacy
  translucencyLightingMode=TLM_VolumetricPerVertexDirectional
  twoSided=true
  usedShadingModels=Default Lit
 }
 nodes [
  n1 = ParticleColor()
  n2 = ComponentMask(None=TextureCoordinate()) {g=true}
  n3 = ScalarParameter() {defaultValue=2, parameterName=GradientAmount}
  n4 = LinearInterpolate(
   A=@n1.RGB
   B=LinearInterpolate(
    A=VectorParameter() {parameterName=Gradient_TopColor, defaultValue=<rgba 0 1 1 1>}.RGB
    B=VectorParameter() {parameterName=Gradient_BottomColor, defaultValue=<rgba 1 0.0005 0 1>}.RGB
    Alpha=Power(Base=@n2, Exp=@n3)
   )
   Alpha=ScalarParameter() {parameterName=IsGradient}
  )
  n5 = TextureCoordinate()
  n6 = ComponentMask(None=@n5) {g=true}
  n7 = ScalarParameter() {defaultValue=2, parameterName=SplineSpherePower}
  n8 = OneMinus(None=ComponentMask(None=@n5) {r=true})
  n9 = StaticSwitchParameter(
   True=SphereMask(A=Power(Base=@n6, Exp=@n7), B=Constant() {r=0.5}, Hardness=ScalarParameter() {parameterName=SphereMaskHardness}) {attenuationRadius=0.5, hardnessPercent=0}
   False=Saturate(None=Multiply(A=@n8) {constB=3})
  ) {parameterName=UseSpline}
  n10 = TextureCoordinate()
  n11 = DynamicParameter() {defaultValue=<rgba 0 1 1 50>, paramNames [Erode,XTileMult,PanAdd,SplineWidth]}
  n12 = Multiply(
   A=StaticSwitchParameter(
    True=MaterialFunctionCall(XY=@n10) {materialFunction=&Swizzle}.YX
    False=TextureCoordinate() {coordinateIndex=1}
   ) {parameterName=UseSpline}
   B=AppendVector(A=@n11.XTileMult, B=NamedRerouteUsage())
  )
  n13 = VectorParameter() {parameterName=MainTileAndPan, defaultValue=<rgba 0.2 0.8 -0.12 0.05>}
  n14 = Add(A=@n11.PanAdd, B=Time())
  n15 = Panner(Coordinate=Multiply(A=@n12, B=AppendVector(A=@n13.R, B=@n13.R)), Time=@n14, Speed=AppendVector(A=@n13.B, B=@n13.B))
  n16 = VectorParameter() {parameterName=SecondaryTileAndPan, defaultValue=<rgba 0.4 1 -0.1 -0.03>}
  n17 = AppendVector(A=@n16.R, B=@n16.R)
  n18 = TextureSample(UVs=Panner(Coordinate=Multiply(A=@n12, B=@n17), Time=@n14, Speed=AppendVector(A=@n16.B, B=@n16.B))) {samplerType=SAMPLERTYPE_Grayscale, texture=/Game/Effects/Fort_Effects/Textures/Fire/T_WispyFireNoise_Soft_M}
  n19 = TextureCoordinate()
  n20 = StaticSwitchParameter(True=ComponentMask(None=@n19) {r=true}, False=ComponentMask(None=@n19) {g=true}) {parameterName=UseSpline}
  n21 = Multiply(A=Subtract(A=@n20) {constB=0.5}) {constB=5}
  n22 = Multiply(A=Multiply(A=@n18.G, B=Clamp(None=@n21) {maxDefault=0.5, minDefault=-0.5})) {constB=-0.4}
  n23 = Add(A=@n15, B=Multiply(A=@n22) {constB=0.25})
  n24 = Multiply(
   A=@n9
   B=Max(
    A=TextureSample(UVs=@n23) {samplerType=SAMPLERTYPE_Grayscale, texture=&T_WispyFireNoise_Soft_M}.G
    B=Multiply(A=@n18.G) {constB=0.3}
   )
  )
  n25 = Add(A=@n24) {constB=0.4}
  MaterialFunctionCall(Emissive=Multiply(A=Multiply(A=@n4, B=@n25), B=VectorParameter() {parameterName=EmissiveMult, defaultValue=<rgba 0.1425 0.1555 0.2 1>}.RGB)) {materialFunction=&MF_FX_EmissiveAdjustments}
  MaterialFunctionCall() {materialFunction=&DebugTimeSine}
  n26 = Multiply(A=@n9, B=SmoothStep(Min=Add(A=@n24) {constB=0.1}, Max=Add(A=@n24) {constB=-0.5}, Value=NamedRerouteUsage()) {constMin=1})
  n27 = SphereMask(A=Add(A=@n22, B=@n20), B=Constant() {r=0.5}) {attenuationRadius=0.3, hardnessPercent=0}
  n28 = Saturate(None=Multiply(A=Multiply(A=@n26, B=@n27), B=ScalarParameter() {defaultValue=1.5, parameterName=BaseOpacityMult}) {constB=2})
  n29 = ScalarParameter() {group=Alpha, parameterName=DepthFadeDistance, sortPriority=2}
  Multiply(A=@n1.A, B=ShadingPathSwitch(Default=DepthFade(Opacity=@n28, FadeDistance=@n29), Mobile=@n28) {desc=Migrated from FeatureLevelSwitch})
  NamedRerouteDeclaration(None=@n11.Erode) {name=Erode, nodeColor=<rgba 0.0547 0 1 1>}
  NamedRerouteDeclaration(None=@n11.SplineWidth) {name=SplineWidth, nodeColor=<rgba 0 1 0.6484 1>}
  n30 = DynamicParameter() {parameterIndex=1, defaultValue=<rgba 0.1 1 1 1>, paramNames [SplineTipMult, YTileMult, Param3, Param4]}
  NamedRerouteDeclaration(None=@n30.YTileMult) {name=YTileMult, nodeColor=<rgba 0 1 1 1>}
  n31 = NamedRerouteUsage()
  n32 = MaterialFunctionCall(WidthBase=@n31, WidthTip=Multiply(A=@n31, B=@n30.SplineTipMult), AdditionalNormal=NamedRerouteUsage()) {materialFunction=&SplineThicken}
  Reroute(None=StaticSwitchParameter(True=@n32.WorldPositionOffset, False=Constant()) {parameterName=UseSpline})
  n33 = Multiply(A=TwoSidedSign(), B=@n32.Normal)
  n34 = MaterialFunctionCall(
   Normal=TextureSample(UVs=@n15) {samplerType=SAMPLERTYPE_Normal, texture=/Game/Effects/Fort_Effects/Textures/Fire/T_WispyFireNoise_Soft_N}.RGB
   Flatness=LinearInterpolate(Alpha=NamedRerouteUsage()) {constB=-2.5}
  ) {materialFunction=&FlattenNormal}
  n35 = TextureSample(UVs=Add(A=TextureCoordinate() {uTiling=0.1}, B=Constant2Vector() {r=0.5})) {samplerType=SAMPLERTYPE_Normal, texture=&T_Sphere_N}
  n36 = Constant() {r=0.25}
  StaticSwitchParameter(
   True=MaterialFunctionCall(
    Normal=Transform(None=@n33) {transformSourceType=TRANSFORMSOURCE_World, transformType=TRANSFORM_Tangent}
    Flatness=ScalarParameter() {defaultValue=0.2, parameterName=SplineNormalFlatness}
   ) {materialFunction=&FlattenNormal}.Result
   False=MaterialFunctionCall(
    BaseNormal=NamedRerouteDeclaration(None=@n34.Result) {name=NormalTexture, nodeColor=<rgba 1 0.5391 0 1>}
    AdditionalNormal=MaterialFunctionCall(Normal=@n35.RGB, Flatness=@n36) {materialFunction=&FlattenNormal}.Result
   ) {materialFunction=&BlendAngleCorrectedNormals}.Result
  ) {parameterName=UseSpline}
 ]
}
