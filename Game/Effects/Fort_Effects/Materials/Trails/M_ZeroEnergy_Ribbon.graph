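# Unlit translucent Niagara ribbon and mesh-particle material for Zero Energy trails, tinted by particle color.
# Two panning star-streak samples, apparently distorted by panning chromatic caustics, layer over a sine-wave blend of two background colors.
# Dynamic parameters drive pan offset, a glowing zip traveling along UV1, and end-fade shift.
# Opacity combines streaks, UV masks and near-camera, depth and optional view-angle fades, which forward and mobile paths replace with a constant.
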Material /Game/Effects/Fort_Effects/Materials/Trails/M_ZeroEnergy_Ribbon
graph {
 properties {
  bUsedWithNiagaraMeshParticles=true
  bUsedWithNiagaraRibbons=true
  blendMode=BLEND_Translucent
  pixelDepthOffsetMode=PDOM_Legacy
  shadingModel=MSM_Unlit
  usedShadingModels=Unlit
 }
 nodes [
  n1 = NamedRerouteUsage()
  n2 = NamedRerouteUsage()
  n3 = VectorParameter() {parameterName=Tile1, defaultValue=<rgba 0.3 0.3 0 0>}
  n4 = Multiply(A=AppendVector(A=@n3.R, B=@n3.R), B=NamedRerouteUsage())
  n5 = VectorParameter() {parameterName=PanDirection1, defaultValue=<rgba 0.05 1 0 0>}
  n6 = AppendVector(A=@n5.R, B=@n5.R)
  n7 = TextureSample(UVs=Panner(Coordinate=Add(A=@n4, B=@n2), Time=NamedRerouteUsage(), Speed=@n6) {speedX=1, speedY=1}) {texture=/Game/Athena/Environments/Nexus/Textures/T_ZeroPoint_Line_Stars}
  n8 = Multiply(A=@n4) {constB=2}
  n9 = VectorParameter() {parameterName=PanDirection1Mult, defaultValue=<rgba -1.7 2.4 0 1>}
  n10 = AppendVector(A=@n9.R, B=@n9.R)
  n11 = TextureSample(UVs=Panner(Coordinate=Subtract(A=@n8, B=@n2), Time=NamedRerouteUsage(), Speed=Multiply(A=@n6, B=@n10)) {speedX=1, speedY=1}) {texture=&T_ZeroPoint_Line_Stars}
  n12 = Multiply(A=@n11.RGB, B=VectorParameter() {parameterName=SecondaryTint, defaultValue=<rgba 0.3664 0.0328 0.7 1>}.RGB) {constB=0.5}
  n13 = NamedRerouteUsage()
  n14 = LinearInterpolate(
   A=VectorParameter() {parameterName=BackgroundColor1, defaultValue=<rgba 0.0396 0.1857 0.2 0>}.RGB
   B=VectorParameter() {parameterName=BackgroundColor2, defaultValue=<rgba 0.0438 0.0296 0.15 0>}.RGB
   Alpha=NamedRerouteUsage()
  )
  n15 = Add(
   A=Add(A=Add(A=@n7.RGB, B=@n12), B=NamedRerouteUsage())
   B=LinearInterpolate(A=Multiply(A=@n13, B=@n14), Alpha=ScalarParameter() {parameterName=Remove BKGD Color}) {constB=0}
  )
  n16 = ParticleColor()
  MaterialFunctionCall(Emissive=Multiply(A=Add(A=@n1, B=@n15), B=@n16.RGB)) {materialFunction=&MF_FX_EmissiveAdjustments}
  n17 = TextureCoordinate() {coordinateIndex=1}
  n18 = Constant() {r=0.5}
  n19 = SphereMask(A=ComponentMask(None=@n17) {g=true}, B=@n18, Radius=@n18) {hardnessPercent=0}
  n20 = Constant() {r=0.3}
  n21 = Add(
   A=NamedRerouteUsage()
   B=NamedRerouteDeclaration(None=MaterialFunctionCall(In=@n19, Contrast=@n20) {materialFunction=&CheapContrast}.Result) {name=UV1 SphereMask Cheap Contrast, nodeColor=<rgba 1 0.1172 0 1>}
  ) {constB=0.5}
  n22 = ComponentMask(None=@n17) {r=true}
  n23 = Power(
   Base=LinearInterpolate(A=@n22, B=OneMinus(None=@n22), Alpha=ScalarParameter() {parameterName=Invert Opacity U})
   Exp=DynamicParameter() {defaultValue=<rgba 1 1 2 1>, paramNames [PanAdd, ZipTravel, EndFadeShift, ""]}.EndFadeShift
  )
  n24 = Constant() {r=0.5}
  n25 = ScalarParameter() {defaultValue=0.6, parameterName=EndFadeHardness}
  n26 = Multiply(A=@n19, B=Saturate(None=SphereMask(A=@n23, B=@n24, Radius=@n24, Hardness=@n25) {hardnessPercent=0}))
  n27 = Multiply(
   A=MaterialFunctionCall(Distance=ScalarParameter() {defaultValue=800, parameterName=NearFadeDIst}) {materialFunction=&Near_Camera_Fade}.Result
   B=DepthFade(FadeDistance=ScalarParameter() {defaultValue=400, parameterName=DepthFade})
  )
  n28 = ScalarParameter() {defaultValue=2, parameterName=AngleFade}
  n29 = LinearInterpolate(
   A=@n27
   B=Multiply(
    A=@n27
    B=MaterialFunctionCall("FadeSensitivity (optional)"=@n28) {materialFunction=&FadeBasedOnViewAngle}.Result
   )
   Alpha=ScalarParameter() {parameterName=FadeBasedOnViewAngle}
  )
  n30 = Constant() {r=0.7}
  Multiply(A=@n16.A, B=Multiply(A=Multiply(A=@n21, B=@n26), B=ShadingPathSwitch(Default=@n29, Forward=@n30, Mobile=@n30)))
  NamedRerouteDeclaration(None=MaterialFunctionCall(XY=TextureCoordinate()) {materialFunction=&Swizzle}.YX) {name=UV0, nodeColor=<rgba 0 0.1328 1 1>}
  n31 = Time()
  n32 = ScalarParameter() {defaultValue=-0.5, parameterName=PanSpeed}
  NamedRerouteDeclaration(None=Add(A=Multiply(A=@n31, B=@n32), B=DynamicParameter() {defaultValue=<rgba 1 1 2 1>, paramNames [PanAdd, ZipTravel, EndFadeShift, ""]}.PanAdd)) {name=PannerTime, nodeColor=<rgba 0 1 0.7891 1>}
  n33 = NamedRerouteUsage()
  n34 = VectorParameter() {parameterName=Tile2, defaultValue=<rgba 0.5 0.025 0 0>}
  n35 = AppendVector(A=@n34.R, B=@n34.R)
  n36 = VectorParameter() {parameterName=PanDirection2, defaultValue=<rgba 0.04 0.1 0 0>}
  n37 = TextureSample(UVs=Panner(Coordinate=Multiply(A=@n33, B=@n35), Time=NamedRerouteUsage(), Speed=AppendVector(A=@n36.R, B=@n36.R)) {speedX=1, speedY=1}) {texture=/Game/Effects/Fort_Effects/Textures/Tile/T_Caustics_Chromatic}
  NamedRerouteDeclaration(None=Multiply(A=ComponentMask(None=@n37.RGB) {g=true, r=true}) {constB=0.05}) {name=AddedBKGD UVs, nodeColor=<rgba 0 1 0.2734 1>}
  NamedRerouteDeclaration(None=Multiply(A=@n37.RGB) {constB=0.5}) {name=AddedBKGD, nodeColor=<rgba 1 0.7266 0 1>}
  NamedRerouteDeclaration(None=@n37.B) {name="AddedBKGD (B)", nodeColor=<rgba 0 1 0.0625 1>}
  n38 = NamedRerouteUsage()
  NamedRerouteDeclaration(None=Add(A=@n7.B, B=Add(A=@n11.B, B=@n38))) {name=Streaks Opacity, nodeColor=<rgba 0 1 0.1328 1>}
  n39 = TextureCoordinate() {coordinateIndex=1}
  n40 = Add(
   A=Divide(A=ComponentMask(None=@n39) {g=true}, B=ScalarParameter() {defaultValue=0.75, parameterName=WaveLength})
   B=Multiply(A=Time(), B=ScalarParameter() {defaultValue=-0.15, parameterName=WaveSpeed})
  )
  NamedRerouteDeclaration(None=ConstantBiasScale(None=Sine(None=@n40))) {name=BKGD_LERPAlpha, nodeColor=<rgba 1 0 0.6797 1>}
  n41 = ScalarParameter() {defaultValue=0.15, parameterName=ZipRadius}
  n42 = Multiply(A=@n41) {constB=0.5}
  n43 = SphereMask(
   A=ComponentMask(None=TextureCoordinate() {coordinateIndex=1}) {r=true}
   B=LinearInterpolate(
    A=Subtract(B=@n42) {constA=0}
    B=Add(A=@n42)
    Alpha=DynamicParameter() {defaultValue=<rgba 1 1 2 1>, paramNames [PanAdd, ZipTravel, EndFadeShift, ""]}.ZipTravel
   ) {constAlpha=1}
   Radius=@n41
  ) {attenuationRadius=0.125, hardnessPercent=0}
  NamedRerouteDeclaration(None=Multiply(A=Power(Base=@n43), B=VectorParameter() {parameterName=ZipGlow, defaultValue=<rgba 0.0481 1.2474 2.5 0>}.RGB)) {name=Zip, nodeColor=<rgba 1 0 0.4219 1>}
 ]
}
